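; Player controller component that tracks the possessed pawn's Boss and Minor Medallions and broadcasts changes.
; On possession it watches a gameplay tag count and the player state's augment list, clearing or unbinding on death or pawn loss.
; Count decreases broadcast immediately, while increases wait for the augment update to identify the new boss medallion.
; Augments are sorted by tag against cached lists, and removed boss medallions are pruned.
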(event EventBeginPlay
  (|CallParentFunction)
  (bind _asplayer_controller (Utilities|Casting|CastToPlayerController (Components|GetOwner)))
  (Pawn|BindEventtoOnPossessedPawnChanged _asplayer_controller (EventDispatchers|CreateEvent :!event Custom|OnPossessedPawnChanged))
  (CallFunction|OnPossessedPawnChanged 0 (Pawn|GetControlledPawn _asplayer_controller)))

(event EventEndPlay
  (params (EndPlayReason :type EEndPlayReason%20Enum :default "Destroyed"))
  (|CallParentFunction EndPlayReason)
  (bind _asplayer_controller (Utilities|Casting|CastToPlayerController (Components|GetOwner)))
  (Pawn|UnbindEventfromOnPossessedPawnChanged _asplayer_controller (EventDispatchers|CreateEvent :!event Custom|OnPossessedPawnChanged)))

(event Custom|OnPossessedPawnChanged
  (params (OldPawn :type Pawn%20Object%20Reference) (NewPawn :type Pawn%20Object%20Reference))
  (Utilities|IsValid NewPawn
    (:Is%20Valid
      (bind _asfort_player_pawn (Utilities|Casting|CastToFortPlayerPawn NewPawn))
      (bind _output_get (Variables|Default|SetCurrentPawn _asfort_player_pawn))
      (bind _output_get_2 (Variables|Default|SetPlayer%20State (Variables|/Script/Engine.Pawn|GetPlayerState _output_get)))
      (EventDispatchers|BindEventtoOnAugmentListUpdated (Class|FortPlayerStateComponentPlayerAugmentSystem|GetAugmentComponent _output_get_2) (EventDispatchers|CreateEvent :!event OnAugmentListUpdated))
      ;[0.2292 0.6077 1] Boss Medallions
      (bind (_asynctaskproxy _tagcount) (Ability|Tasks|WaitGameplayTagCountChangedOnActor (Variables|Default|GetCurrentPawn) (Variables|BossMedallions|GetMedallionTrackingTag))
        (:then
          (Variables|BossMedallions|SetWaitForTrackingTag _asynctaskproxy))
        (:TagCountChanged
          (if (FortPawn|IsDead (Variables|Default|GetCurrentPawn))
            (BossMedallions|ClearMedallionListener)
            (elif (!= _tagcount (Variables|BossMedallions|GetCurrentMedallionCount))
              (bind _output_get_3 (Variables|BossMedallions|SetWasLastChangeIncrease (> _tagcount (Variables|BossMedallions|GetCurrentMedallionCount))))
              (Variables|BossMedallions|SetCurrentMedallionCount _tagcount)
              (if (not _output_get_3)
                (Default|CallOnMedallionCountChanged false _tagcount)))))))
    (:Is%20Not%20Valid
      (Utilities|IsValid OldPawn
        (:Is%20Valid
          (BossMedallions|ClearMedallionListener)
          (Default|UnbindfromAugmentListUpdated))) ; Only need to unbind if we had a pawn
    )))

(fn ClearMedallionListener
  (AsyncAction|Cancel (Variables|BossMedallions|GetWaitForTrackingTag))
  (Variables|BossMedallions|SetCurrentMedallionCount 0))

(fn ForceBroadcastNewMinorMedallion
  (params (NewMinorMedallion :type Fort%20Player%20Augment%20Item%20Definition%20Object%20Reference%20%28by%20ref%29))
  (Utilities|Array|Add (Variables|MinorMedallions|GetActiveMinorMedallions) NewMinorMedallion)
  (bind _output_get (Variables|MinorMedallions|SetNewestMinorMedallion NewMinorMedallion))
  (Default|CallOnMinorMedallionAdded _output_get))

(fn GetCurrentMedallionCount
  (return (Variables|BossMedallions|GetCurrentMedallionCount)))

(fn GetNewestMedallionAdded
  (return (Variables|BossMedallions|GetNewestBossMedallion)))

(fn GetNewestMinorMedallion
  (return (Variables|MinorMedallions|GetNewestMinorMedallion)))

(fn GetWasLastMedallionChangeIncrease
  (return (Variables|BossMedallions|GetWasLastChangeIncrease)))

(fn OnAugmentListUpdated
  (params (AugmentComponent :type Fort%20Player%20State%20Component%20Player%20Augment%20System%20Object%20Reference))
  (bind _temp_augmenttocheck (LocalVariables|GetTemp_AugmentToCheck))
  (bind _temp_augmenttocheck_2 (LocalVariables|GetTemp_AugmentToCheck))
  (if (FortPawn|IsDead (Variables|Default|GetCurrentPawn))
    (Default|UnbindfromAugmentListUpdated)
    (else
      (bind _returnvalue (Class|FortPlayerStateComponentPlayerAugmentSystem|GetActiveAugments (Variables|Default|GetPlayer%20State)))
      (bind _output_get (LocalVariables|SetAllPlayerAugments _returnvalue))
      (for _array_element _output_get
        (bind _output_get_2 (LocalVariables|SetTemp_AugmentToCheck _array_element))
        ;[0.2292 0.6077 1] Check for newest Boss Medallion
        (if (GameplayTags|HasMatchingGameplayTag _output_get_2 (Variables|BossMedallions|GetBossMedallionTag))
          (if (not (Utilities|Array|ContainsItem (Variables|BossMedallions|GetUpdatedBossMedallions) (LocalVariables|GetTemp_AugmentToCheck)))
            (Utilities|Array|Add (Variables|BossMedallions|GetUpdatedBossMedallions) _temp_augmenttocheck)
            (bind _output_get_3 (Variables|BossMedallions|SetNewestBossMedallion _temp_augmenttocheck))
            (Default|CallOnMedallionCountChanged (BossMedallions|GetWasLastMedallionChangeIncrease) (BossMedallions|GetCurrentMedallionCount) _output_get_3))
          (else
            ;[0.2656 1 0.3389] Check for any new Minor Medallions
            (if (GameplayTags|HasMatchingGameplayTag (LocalVariables|GetTemp_AugmentToCheck) (Variables|MinorMedallions|GetMinorMedallionTag))
              (if (not (Utilities|Array|ContainsItem (Variables|MinorMedallions|GetActiveMinorMedallions) (LocalVariables|GetTemp_AugmentToCheck)))
                (Utilities|Array|Add (Variables|MinorMedallions|GetActiveMinorMedallions) _temp_augmenttocheck_2)
                (bind _output_get_4 (Variables|MinorMedallions|SetNewestMinorMedallion _temp_augmenttocheck_2))
                (Default|CallOnMinorMedallionAdded _output_get_4))))))
      ;[0.2292 0.6077 1] Refresh Boss Medallion list in case one was removed
      (for _array_element_2 (Variables|BossMedallions|GetUpdatedBossMedallions)
        (if (not (Utilities|Array|ContainsItem (LocalVariables|GetAllPlayerAugments) _array_element_2))
          (Utilities|Array|RemoveItem (Variables|BossMedallions|GetUpdatedBossMedallions) _array_element_2))))))

(fn UnbindFromAugmentListUpdated
  (EventDispatchers|UnbindEventfromOnAugmentListUpdated (Class|FortPlayerStateComponentPlayerAugmentSystem|GetAugmentComponent (Variables|Default|GetPlayer%20State)) (EventDispatchers|CreateEvent :!event OnAugmentListUpdated))
  (Variables|MinorMedallions|SetNewestMinorMedallion 0))
